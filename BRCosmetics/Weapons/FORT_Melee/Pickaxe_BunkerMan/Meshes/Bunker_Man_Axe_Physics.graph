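# Physics asset for the Bunker Man two-handed melee pickaxe, giving its dangling ribbon a physics setup.
# One large capsule covers the handle, and small capsules cover the ribbon base and two three-bone strands, a and b.
# Each strand is a constraint chain branching from the base, with equal swing limits and tighter twist limits.
# No constraint links the ribbon base to the handle body.
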PhysicsAsset /BRCosmetics/Weapons/FORT_Melee/Pickaxe_BunkerMan/Meshes/Bunker_Man_Axe_Physics
bodies {
 Melee_TwoHanded_Handle [{shapeType=Capsule, length=74.3464, radius=20.4205, center=<0.0004 2.4547 27.4218>}]
 dyn_ribbon_a01 [{shapeType=Capsule, length=2.8319, radius=1.5766, center=<1.8291 -0.6302 0.0001>, rotation=<rot 90 11.7683 6.7685>}]
 dyn_ribbon_a02 [{shapeType=Capsule, length=5.6522, radius=1.7296, center=<2.8262 0.0002 0>, rotation=<rot 90 0.0396 0.0397>}]
 dyn_ribbon_a03 [{shapeType=Capsule, length=5.3164, radius=2.0653, center=<3.0088 -0.0001 -0.2641>, rotation=<rot 84.9999 0 0>}]
 dyn_ribbon_b01 [{shapeType=Capsule, length=3.6677, radius=1.0586, center=<2.267 -0.0001 0>, rotation=<rot 90 35.2644 35.2634>}]
 dyn_ribbon_b02 [{shapeType=Capsule, length=2.7255, radius=1.6697, center=<1.1444 0 0.8507>, rotation=<rot 79.993 179.9999 179.9999>}]
 dyn_ribbon_b03 [{shapeType=Capsule, length=3.3781, radius=1.6697, center=<1.6773 0.0002 0.0159>, rotation=<rot 84.9869 -180 -180>}]
 dyn_ribbon_base [{shapeType=Capsule, length=0.7701, radius=1.1754, center=<-0.1477 0 0>, rotation=<rot 0 -89.9998 -89.9992>}]
}
constraints [
 {bone1Name=dyn_ribbon_b01, bone2Name=dyn_ribbon_base, swing1LimitDegrees=60, swing2LimitDegrees=60, twistLimitDegrees=15}
 {bone1Name=dyn_ribbon_b02, bone2Name=dyn_ribbon_b01, swing1LimitDegrees=60, swing2LimitDegrees=60, twistLimitDegrees=15}
 {bone1Name=dyn_ribbon_b03, bone2Name=dyn_ribbon_b02, swing1LimitDegrees=60, swing2LimitDegrees=60, twistLimitDegrees=15}
 {bone1Name=dyn_ribbon_a01, bone2Name=dyn_ribbon_base, swing1LimitDegrees=60, swing2LimitDegrees=60, twistLimitDegrees=5}
 {bone1Name=dyn_ribbon_a02, bone2Name=dyn_ribbon_a01, swing1LimitDegrees=60, swing2LimitDegrees=60, twistLimitDegrees=15}
 {bone1Name=dyn_ribbon_a03, bone2Name=dyn_ribbon_a02, swing1LimitDegrees=60, swing2LimitDegrees=60, twistLimitDegrees=15}
]
